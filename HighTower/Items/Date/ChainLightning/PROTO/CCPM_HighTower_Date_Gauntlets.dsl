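; Character part Blueprint for the HighTower Date chain lightning gauntlets, handling their effects and firing animation.
; On part change it attaches idle and repulsor muzzle Niagara effects, registers itself with the weapon and binds to overheat changes.
; Overheat events drive idle effect parameters and a gated recharge flash.
; Firing activates that hand's muzzle effect, briefly swaps emissive color and toggles firing flags on the linked anim layer.
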(event EventOnPawnAndPartChanged
  (bind _returnvalue (CustomCharacter|GetAssociatedPlayerPawn))
  (bind _returnvalue_2 (Weapon|GetCurrentWeapon _returnvalue))
  (|CallParentFunction)
  (CallFunction|MuzzleAttachment)
  (CallFunction|IdleAttachment)
  (Utilities|IsValid _returnvalue
    (:Is%20Valid
      (CallFunction|HandleNewWeapon _returnvalue_2)
      (bind _asb_high_tower_date_combo (Utilities|Casting|CastToB_HighTower_Date_Combo _returnvalue_2))
      (Overheating|AssignOnOverheatValueChanged _asb_high_tower_date_combo (EventDispatchers|CreateEvent :!event Custom|OnOverheatValueChanged_Event)))))

(event EventEndPlay
  (params (EndPlayReason :type EEndPlayReason%20Enum :default "Destroyed"))
  (bind _returnvalue (CustomCharacter|GetAssociatedPlayerPawn))
  ;[1 0.4712 0.05898 0.5] EndPlay
  (|CallParentFunction EndPlayReason)
  (Utilities|IsValid _returnvalue
    (:Is%20Valid
      (bind _asb_high_tower_date_combo (Utilities|Casting|CastToB_HighTower_Date_Combo (Weapon|GetCurrentWeapon _returnvalue)))
      (Overheating|UnbindEventfromOnOverheatValueChanged _asb_high_tower_date_combo (EventDispatchers|CreateEvent :!event Custom|OnOverheatValueChanged_Event)))))

(event Custom|OnOverheatValueChanged_Event
  (params (CurrentValue :type Float%20%28single-precision%29) (MaxValue :type Float%20%28single-precision%29) (State :type EFortWeaponOverheatState%20Enum) (bIsInCooldown :type Boolean))
  (bind _ns_idle (Variables|Default|GetNS_Idle))
  ; Overheat FX
  (Utilities|FlowControl|Sequence
    (:then_0
      (Effects|Components|ParticleSystem|SetBoolParameter _ns_idle "IsCooldown" (and bIsInCooldown (!= CurrentValue 0)))
      (Effects|Components|ParticleSystem|SetFloatParameter _ns_idle "Overheat" CurrentValue)
      (if (not bIsInCooldown)
        (Effects|Components|ParticleSystem|SetBoolParameter _ns_idle "IsOverheating" false))
      (Utilities|FlowControl|DoOnce false
        (:Completed
          (Effects|Components|ParticleSystem|SetBoolParameter _ns_idle "IsOverheating" true))))
    (:then_1
      (if (== CurrentValue 0)
        _
        (elif (> CurrentValue 91.0)
          _))
      (Utilities|FlowControl|DoOnce true
        (:Completed
          (bind _fx_rechargeflash (Variables|Default|GetFX_RechargeFlash)
            (:then
              (Components|Activation|Activate _fx_rechargeflash true))
            (:else
              (bind _returnvalue (Effects|SpawnFXSystemAttached "/HighTower/Items/Date/ChainLightning/FX/NS_HighTower_Date_ChainLightning_Gloves_RechargeFlash.NS_HighTower_Date_ChainLightning_Gloves_RechargeFlash" (Variables|/Script/Engine.Character|GetMesh (CustomCharacter|GetAssociatedPlayerPawn))))
              (Variables|Default|SetFX_RechargeFlash _returnvalue))))))))

(event Custom|Idle%20Attachment
  (Transformation|AttachComponentToComponent (Variables|Default|GetNS_Idle) (Customization|GetSkeletalMeshforPartType (CustomCharacter|GetAssociatedPlayerPawn) (CustomCharacter|GetCharacterPartType) :!class /Script/FortniteGame.FortPlayerPawn) "None" "KeepRelative" "KeepRelative" "KeepRelative" true)
  ;[1 1 1] Cancel Auto Activate of Idle FX, add set active to fix JIRA FORT-308838
  (Components|Activation|SetActive (Variables|Default|GetNS_Idle) true true))

(event EventOnWeaponChanged
  (params (NewWeapon :type Fort%20Weapon%20Object%20Reference) (OldWeapon :type Fort%20Weapon%20Object%20Reference))
  ;[1 0 0] Unused
  (Utilities|IsValid NewWeapon
    (:Is%20Valid
      (CallFunction|HandleNewWeapon NewWeapon))
    (:Is%20Not%20Valid
      (Utilities|IsValid OldWeapon
        (:Is%20Valid
          (CallFunction|HandleNewWeapon OldWeapon))
        (:Is%20Not%20Valid
          (Development|PrintString "CCPM_HighTower_Date_Gauntlets: No valid weapon, that's not good!"))))))

(event Custom|HandleNewWeapon
  (params (WeaponObject :type Fort%20Weapon%20Object%20Reference))
  ;[1 0.6369 0.427] HandleNewWeapon
  (Utilities|FlowControl|Sequence
    (:then_0
      ;;[1 0.6261 0 0.5] Remove this component from the previous weapon.
      (bind _ourcurrentweapon (Variables|Default|GetOurCurrentWeapon)
        (:then
          (Class|BHighTowerDateChainLightning|ClearCustomCharacterPart _ourcurrentweapon self))
        (:else)))
    (:then_1
      ;;[0.193 0.9844 0.1077 0.5] Register this CCPM with the weapon (the weapon may need it to play animations.)
      (bind _asb_high_tower_date_chain_lightning (Utilities|Casting|CastToB_HighTower_Date_ChainLightning WeaponObject))
      (bind _output_get (Variables|Default|SetOurCurrentWeapon _asb_high_tower_date_chain_lightning))
      (Class|BHighTowerDateChainLightning|SetCustomCharacterPart _output_get self :!class /HighTower/Items/Date/ChainLightning/B_HighTower_Date_ChainLightning.B_HighTower_Date_ChainLightning_C))))

(event Custom|Muzzle_Attachment
  (bind _returnvalue (CustomCharacter|GetSkeletalMeshForPartOnPlayerPawn))
  ;[1 0.6369 0.427] Muzzle Setups
  (Transformation|AttachComponentToComponent (Variables|Default|GetNS_MuzzleLeft) _returnvalue "FX_repulsor_l" "KeepRelative" "KeepRelative" "KeepRelative" true)
  (Transformation|AttachComponentToComponent (Variables|Default|GetNS_MuzzleRight) _returnvalue "FX_repulsor_r" "KeepRelative" "KeepRelative" "KeepRelative" true)
  (if (> (Math|Vector|VectorLength (Transformation|GetWorldScale (Variables|Default|GetNS_MuzzleLeft))) 1e+01)
    (Transformation|SetWorldScale3D (Variables|Default|GetNS_MuzzleRight) "10.000000,10.000000,10.000000")))

(event Custom|StartFiringAnimation
  (params (FiringHand :type EDualWeaponHand%20Enum :default "LEFT"))
  (bind _returnvalue (CustomCharacter|GetCharacterPartType))
  (bind _returnvalue_2 (CustomCharacter|GetAssociatedPlayerPawn))
  (bind _returnvalue_3 (Customization|GetSkeletalMeshforPartType _returnvalue_2 _returnvalue :!class /Script/FortniteGame.FortPlayerPawn))
  ;[1 0.6369 0.427] StartFiringAnimation
  (bind _output_get (Variables|Default|SetCurrentlyFiringHand FiringHand))
  (CallFunction|UpdateFiringAnimInstance _output_get true)
  (Components|Activation|Activate (Utilities|Select :Index _output_get :LEFT (Variables|Default|GetNS_MuzzleLeft) :RIGHT (Variables|Default|GetNS_MuzzleRight)) false)
  (Rendering|Material|SetVectorParameterValueonMaterials _returnvalue_3 "Emissive Color" (Math|Conversions|ToVector(LinearColor) (Variables|Default|GetLMB%20Color)))
  (Utilities|FlowControl|Delay 0.3)
  (Rendering|Material|SetVectorParameterValueonMaterials _returnvalue_3 "Emissive Color" (Math|Conversions|ToVector(LinearColor) (Variables|Default|GetBoostColor)))
  (CallFunction|UpdateFiringAnimInstance (Variables|Default|GetCurrentlyFiringHand)))

(fn UpdateFiringAnimInstance
  (params (Hands :type EDualWeaponHand%20Enum :default "LEFT") (IsFiring :type Boolean))
  (bind _ashigh_tower_date_layer_anim_bp (Utilities|Casting|CastToHighTower_Date_LayerAnimBP (Components|SkeletalMesh|AnimationBlueprintLinking|GetLinkedAnimLayerInstancebyClass (Variables|/Script/Engine.Character|GetMesh (CustomCharacter|GetAssociatedPlayerPawn)) "/HighTower/Animation/Date/HighTower_Date_LayerAnimBP.HighTower_Date_LayerAnimBP_C")))
  (switch Utilities|FlowControl|Switch|SwitchonEDualWeaponHand Hands
    (:LEFT
      (Variables|/Script/HighTowerRuntime.HighTowerDateLayerAnimInstance|SetbIsFiringLeft IsFiring _ashigh_tower_date_layer_anim_bp))
    (:RIGHT
      (Variables|/Script/HighTowerRuntime.HighTowerDateLayerAnimInstance|SetbIsFiringRight IsFiring _ashigh_tower_date_layer_anim_bp))))

(fn ConstructionScript)
